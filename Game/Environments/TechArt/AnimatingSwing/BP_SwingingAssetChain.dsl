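; Blueprint that builds a swinging chain of static meshes, with the motion driven by materials rather than Tick.
; The construction script adds a mesh component per Assets entry, using matching Transforms and optional slot 0 material overrides.
; Dynamic material instances, always created on slot 0, receive a scale-corrected pivot, a rotation axis from OriginAndDirection, and Speed, Angle and TimeOffset through FanRotation oscillation parameters.
; BeginPlay, overlap and Tick events are empty.
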(event EventBeginPlay)

(event Collision|EventActorBeginOverlap
  (params (OtherActor :type Actor%20Object%20Reference)))

(event EventTick
  (params (DeltaSeconds :type Float%20%28single-precision%29)))

(fn ConstructionScript
  (bind _materialoverrides (Variables|Default|GetMaterialOverrides))
  (bind _originanddirection (Variables|Default|GetOriginAndDirection))
  (for (_array_element _array_index) (Variables|Default|GetAssets)
    (bind _returnvalue (AddComponent|NODE_AddStaticMeshComponent-0 :RelativeTransform (Utilities|Array|Get(acopy) (Variables|Default|GetTransforms) _array_index)))
    (Components|StaticMesh|SetStaticMesh _returnvalue _array_element)
    (Utilities|FlowControl|Sequence
      (:then_0
        (if (Utilities|IsValid (Utilities|Array|Get(acopy) _materialoverrides _array_index) :!class /Script/Engine.KismetSystemLibrary)
          (Rendering|Material|SetMaterial _returnvalue 0 (Utilities|Array|Get(acopy) _materialoverrides _array_index))))
      (:then_1
        (for _ (Rendering|Material|GetMaterials _returnvalue)
          (bind _returnvalue_2 (Rendering|Material|CreateDynamicMaterialInstance _returnvalue 0 0 "None" :!class /Script/Engine.PrimitiveComponent))
          (Rendering|Material|SetVectorParameterValue _returnvalue_2 "PivotPointBias" (Math|Conversions|ToLinearColor(Vector) (- (.location _originanddirection) (/ (.location (Transformation|GetRelativeTransform _returnvalue)) (.scale (Transformation|GetRelativeTransform _returnvalue))))) :!class /Script/Engine.MaterialInstanceDynamic)
          (Rendering|Material|SetVectorParameterValue _returnvalue_2 "NormalizedRotationAxis" (Math|Conversions|ToLinearColor(Vector) (Math|Vector|GetForwardVector (.rotation _originanddirection))) :!class /Script/Engine.MaterialInstanceDynamic)
          (Rendering|Material|SetScalarParameterValue _returnvalue_2 "FanRotationSpeed" (/ (Variables|Default|GetSpeed) 1e+02) :!class /Script/Engine.MaterialInstanceDynamic)
          (Rendering|Material|SetScalarParameterValue _returnvalue_2 "FanRotation_OscillatingAngle" (/ (Variables|Default|GetAngle) 1e+02) :!class /Script/Engine.MaterialInstanceDynamic)
          (Rendering|Material|SetScalarParameterValue _returnvalue_2 "FanRotation_OscillatingPause" 1.0 :!class /Script/Engine.MaterialInstanceDynamic)
          (Rendering|Material|SetScalarParameterValue _returnvalue_2 "FanRotation_Randomess" 0.0 :!class /Script/Engine.MaterialInstanceDynamic)
          (Rendering|Material|SetScalarParameterValue _returnvalue_2 "FanRotation_ManualTimeOffset" (Variables|Default|GetTimeOffset) :!class /Script/Engine.MaterialInstanceDynamic))))))
